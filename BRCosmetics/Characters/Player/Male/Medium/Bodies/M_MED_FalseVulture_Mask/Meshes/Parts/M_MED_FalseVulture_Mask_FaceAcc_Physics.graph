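# Physics asset for the face accessory of the male medium False Vulture Mask cosmetic, covering its center, left and right dyn_hair bones.
# Each chain is a capsule on its 01 bone, constrained to a small sphere on its MSTR bone with tight twist and a much wider second swing axis.
# Unconstrained capsules on both clavicles and spine_05 approximate the shoulders and upper torso, apparently as colliders for the dangling pieces.
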PhysicsAsset /BRCosmetics/Characters/Player/Male/Medium/Bodies/M_MED_FalseVulture_Mask/Meshes/Parts/M_MED_FalseVulture_Mask_FaceAcc_Physics
bodies {
 clavicle_l [{shapeType=Capsule, length=12, radius=5.5, shapeName=clavicle_l_capsule, center=<11.3729 0 1.3871>, rotation=<rot -90 160.529 -160.5295>}]
 clavicle_r [{
   shapeType=Capsule
   length=9.5
   radius=5.5
   shapeName=clavicle_r_capsule
   center=<-8.8972 0.0002 -1.3454>
   rotation=<rot -85 179.9968 -179.9983>
  }]
 dyn_hair_01 [{shapeType=Capsule, length=4.5, radius=2, shapeName=dyn_hair_01_capsule, center=<11.318 0 0.6616>, rotation=<rot -65.0012 0 -90>}]
 dyn_hair_MSTR [{shapeType=Sphere, radius=1, shapeName=dyn_hair_MSTR_sphere}]
 dyn_hair_l_01 [{
   shapeType=Capsule
   length=4.5
   radius=2
   shapeName=dyn_hair_l_01_capsule
   center=<9.2651 -0.5421 1.0356>
   rotation=<rot 17.8818 42.3722 -67.249>
  }]
 dyn_hair_l_MSTR [{shapeType=Sphere, radius=1, shapeName=dyn_hair_l_MSTR_sphere}]
 dyn_hair_r_01 [{
   shapeType=Capsule
   length=3
   radius=2
   shapeName=dyn_hair_r_01_capsule
   center=<10.3787 0.7593 -0.0287>
   rotation=<rot -53.5322 -25.7096 -74.0098>
  }]
 dyn_hair_r_MSTR [{shapeType=Sphere, radius=1, shapeName=dyn_hair_r_MSTR_sphere}]
 spine_05 [{shapeType=Capsule, length=20, radius=10, shapeName=spine_05_capsule, center=<2.7254 -0.8874 0>}]
}
constraints [
 {bone1Name=dyn_hair_01, bone2Name=dyn_hair_MSTR, swing1LimitDegrees=15, swing2LimitDegrees=75, twistLimitDegrees=5}
 {bone1Name=dyn_hair_l_01, bone2Name=dyn_hair_l_MSTR, swing1LimitDegrees=25, swing2LimitDegrees=75, twistLimitDegrees=5}
 {bone1Name=dyn_hair_r_01, bone2Name=dyn_hair_r_MSTR, swing1LimitDegrees=25, swing2LimitDegrees=75, twistLimitDegrees=5}
]
